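# Material function in the front-end backplate textures folder that blends four inputs, A through D, along a single 0-1 Alpha, exposed to the library.
# Alpha is tripled, then offset by successive subtractions and saturated to give three segment weights.
# These weights drive three chained lerps, moving from A to B to C to D in equal thirds without branching.
# The built-in description mentions three inputs, but the graph blends four.
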MaterialFunction /Game/Environments/FrontEnd/BackPlates/Textures/Delulu/Lerp_4Color
graph {
 properties {bExposeToLibrary=true, description=Lerps 3 inputs based on a single value.}
 nodes [
  n1 = FunctionInput() {
   bUsePreviewValueAsDefault=true
   description=0-1 Lerp Alpha
   inputName=Alpha
   inputType=FunctionInput_Scalar
   sortPriority=4
   previewValue=<0.5 0 0 1>
  }
  n2 = Multiply(A=@n1) {constB=3}
  n3 = LinearInterpolate(
   A=FunctionInput() {bUsePreviewValueAsDefault=true, inputName=A}
   B=FunctionInput() {bUsePreviewValueAsDefault=true, inputName=B, sortPriority=1, previewValue=<0.5 0.5 0.5 1>}
   Alpha=Saturate(None=@n2)
  )
  n4 = FunctionInput() {bUsePreviewValueAsDefault=true, inputName=C, sortPriority=2, previewValue=<1 1 1 1>}
  n5 = Subtract(A=@n2)
  n6 = Saturate(None=@n5)
  n7 = Subtract(A=@n5)
  FunctionOutput(None=LinearInterpolate(
    A=LinearInterpolate(A=@n3, B=@n4, Alpha=@n6)
    B=FunctionInput() {bUsePreviewValueAsDefault=true, inputName=D, sortPriority=3, previewValue=<1 1 1 1>}
    Alpha=Saturate(None=@n7)
   ))
 ]
}
