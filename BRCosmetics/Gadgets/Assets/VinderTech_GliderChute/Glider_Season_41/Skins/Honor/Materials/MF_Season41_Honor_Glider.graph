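# Material function for the Season 41 Honor glider skin that adds animated, warm-colored artwork to incoming material attributes.
# Packed texture channels animate via polar panning (R, G) or rotation (B, A), with color ramps, UV1 box-mask gradients and pulsing edge glow.
# A halftone pattern uses object-stabilized screen-space UVs and derivative-antialiased distance-field dots.
# The emissive composite splits between BaseColor and Emissive, alongside adjusted Roughness and Specular.
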MaterialFunction /BRCosmetics/Gadgets/Assets/VinderTech_GliderChute/Glider_Season_41/Skins/Honor/Materials/MF_Season41_Honor_Glider
graph {nodes [
  n1 = Time()
  n2 = VectorParameter() {group=CopperMoggy, parameterName=SoftNoiseSpeed_A, defaultValue=<rgba 0.2 0.4 0 0>}
  n3 = AppendVector(A=@n2.R, B=@n2.R)
  n4 = Constant() {r=20}
  n5 = Floor(None=Multiply(A=Multiply(A=@n1, B=@n3), B=@n4))
  n6 = Multiply(A=Add(A=NamedRerouteUsage(), B=Divide(A=@n5, B=@n4)), B=@n4)
  n7 = Multiply(A=Divide(A=Floor(None=@n6), B=@n4), B=ScalarParameter() {defaultValue=1, group=CopperMoggy, parameterName=SoftNoiseTiling_A})
  n8 = VectorParameter() {group=CopperMoggy, parameterName=SoftNoiseSpeed_A, defaultValue=<rgba 0.2 0.4 0 0>}
  Add(A=@n7, B=Multiply(A=AppendVector(A=@n8.R, B=@n8.R), B=Time()))
  n9 = Time()
  n10 = VectorParameter() {group=CopperMoggy, parameterName=SoftNoiseSpeed_B, defaultValue=<rgba -0.2 0.4 0 0>}
  n11 = AppendVector(A=@n10.R, B=@n10.R)
  n12 = Floor(None=Multiply(A=Multiply(A=@n9, B=@n11), B=@n4))
  n13 = Multiply(A=Add(A=NamedRerouteUsage(), B=Divide(A=@n12, B=@n4)), B=@n4)
  n14 = Multiply(A=Divide(A=Floor(None=@n13), B=@n4), B=ScalarParameter() {defaultValue=1, group=CopperMoggy, parameterName=SoftNoiseTiling_B})
  n15 = VectorParameter() {group=CopperMoggy, parameterName=SoftNoiseSpeed_B, defaultValue=<rgba -0.2 0.4 0 0>}
  Add(A=@n14, B=Multiply(A=AppendVector(A=@n15.R, B=@n15.R), B=Time()))
  n16 = GetMaterialAttributes(None=NamedRerouteUsage()) {attributeGetTypes [#1, #2]}
  n17 = MaterialFunctionCall!(
   A=TextureCoordinate() {coordinateIndex=1}
   B=Constant2Vector() {g=0.5, r=0.5}
   Bounds=Constant2Vector() {g=0.2, r=0.2}
   Edge Falloff=Constant() {r=0.4}
  ) {materialFunction=/Engine/Functions/Engine_MaterialFunctions01/Gradient/BoxMask-2D}
  n18 = NamedRerouteDeclaration(None=LinearInterpolate(
    A=VectorParameter() {parameterName=GradientColorA, defaultValue=<rgba 0.087 0.05 0.0484 1>}.RGB
    B=VectorParameter() {parameterName=GradientColorB, defaultValue=<rgba 2 0.4924 0.0546 1>}.RGB
    Alpha=NamedRerouteDeclaration(None=@n17.Result) {name=BoxMask, nodeColor=<rgba 0.1484 0 1 1>}
   )) {name=Gradient, nodeColor=<rgba 0.8984 1 0 1>}
  n19 = NamedRerouteUsage()
  n20 = NamedRerouteUsage()
  n21 = NamedRerouteUsage()
  n22 = NamedRerouteUsage()
  n23 = NamedRerouteUsage()
  n24 = NamedRerouteUsage()
  n25 = NamedRerouteDeclaration(None=LinearInterpolate(A=Multiply(A=@n19, B=@n20), B=Add(A=@n21, B=@n22), Alpha=Add(A=@n23, B=@n24))) {name=AnimatedTexture, nodeColor=<rgba 0 0.5781 1 1>}
  n26 = LinearInterpolate(
   A=VectorParameter() {parameterName=BackgroundColorA, defaultValue=<rgba 0.0612 0.0132 0.0096 1>}.RGB
   B=LinearInterpolate(A=NamedRerouteUsage(), B=Multiply(A=@n18, B=@n25), Alpha=NamedRerouteUsage())
   Alpha=MaterialFunctionCall(
    A=TextureCoordinate() {coordinateIndex=1}
    B=Constant2Vector() {g=0.5, r=0.5}
    Bounds=Constant2Vector() {g=0.45, r=0.45}
    Edge Falloff=Constant() {r=0.3}
   ) {materialFunction=&BoxMask-2D}.Result
  )
  n27 = NamedRerouteUsage()
  n28 = NamedRerouteDeclaration(None=Add(A=Add(A=@n26, B=@n27), B=NamedRerouteUsage())) {name=Out_Emissive, nodeColor=<rgba 1 0.0703 0 1>}
  n29 = NamedRerouteUsage()
  n30 = SetMaterialAttributes!(
   MaterialAttributes=@n16.MaterialAttributes
   BaseColor=Multiply(A=@n28) {constB=0.2}
   EmissiveColor=Multiply(A=@n28) {constB=0.8}
   Roughness=Saturate(None=Multiply(A=@n29) {constB=2})
   Specular=Multiply(A=NamedRerouteUsage()) {constB=0.5}
  ) {attributeSetTypes [
    #3
    #4
    #1
    #2
   ]}
  FunctionOutput(None=@n30) {outputName=MaterialAttributes}
  n31 = ComponentMask(None=NamedRerouteUsage()) {g=true}
  LinearInterpolate(A=@n31, B=Subtract(A=@n31, B=NamedRerouteUsage())) {constAlpha=0.3}
  Multiply(A=@n23, B=@n24)
  n32 = NamedRerouteUsage()
  n33 = NamedRerouteUsage()
  Multiply(A=@n32, B=@n33)
  n34 = Time()
  n35 = VectorParameter() {parameterName=RedChannelSpeedAndTiling, defaultValue=<rgba -0.1 0 2 4>}
  n36 = ComponentMask(None=@n35.RGB) {g=true, r=true}
  n37 = ComponentMask(None=@n35.RGBA) {a=true, b=true}
  n38 = MaterialFunctionCall!(
   UV_Input=TextureCoordinate() {coordinateIndex=1}
   PolarUV_CenterOffsetG
   PolarUV_CenterOffsetR
   PolarUV_LenghtScale=ScalarParameter() {defaultValue=0.714, parameterName=Polar_LengthScaleRedChannel}
   PolarUV_RadialScale=ScalarParameter() {defaultValue=2, parameterName=Polar_RadialScaleRedChannel}
   PolarUV_UseTwist=StaticSwitch()
   PolarUV_TwistOffset
   PolarUVTwistSpeed
   PolarUV_TwistRadius
  ) {materialFunction=/Game/Global/Materials/MaterialGym/Materials/MaterialElements/UVs/UVMainpulation/PolarUV/MF_Polar}
  n39 = TextureCoordinate() {coordinateIndex=1}
  n40 = TextureSample!(
   UVs=Add(A=Multiply(A=@n34, B=@n36), B=Multiply(A=@n37, B=@n38.Result))
   Tex
   "DDX(UVs)"=DDX(Value=@n39)
   "DDY(UVs)"=DDY(Value=@n39)
   Apply View MipBias
  ) {
   mipValueMode=TMVM_Derivative
   samplerSource=SSM_Wrap_WorldGroupSettings
   samplerType=SAMPLERTYPE_LinearColor
   texture=/BRCosmetics/Gadgets/Assets/VinderTech_GliderChute/Glider_Season_41/Skins/Honor/Textures/T_Glider_Season41
  }
  NamedRerouteDeclaration(None=@n40.R) {name=RedChannelPolarTexture, nodeColor=<rgba 1 0 0 1>}
  n41 = Time()
  n42 = VectorParameter() {parameterName=GreenChannelSpeedAndTiling, defaultValue=<rgba -0.1 0 2 2>}
  n43 = ComponentMask(None=@n42.RGB) {g=true, r=true}
  n44 = ComponentMask(None=@n42.RGBA) {a=true, b=true}
  n45 = MaterialFunctionCall(
   UV_Input=TextureCoordinate() {coordinateIndex=1}
   PolarUV_LenghtScale=ScalarParameter() {defaultValue=0.714, parameterName=Polar_LengthScaleGreenChannel}
   PolarUV_RadialScale=ScalarParameter() {defaultValue=2, parameterName=Polar_RadialScaleGreenChannel}
   PolarUV_UseTwist=StaticSwitch()
  ) {materialFunction=&MF_Polar}
  n46 = TextureCoordinate() {coordinateIndex=1}
  n47 = TextureSample(UVs=Add(A=Multiply(A=@n41, B=@n43), B=Multiply(A=@n44, B=@n45.Result)), "DDX(UVs)"=DDX(Value=@n46), "DDY(UVs)"=DDY(Value=@n46)) {
   mipValueMode=TMVM_Derivative
   samplerSource=SSM_Wrap_WorldGroupSettings
   samplerType=SAMPLERTYPE_LinearColor
   texture=&T_Glider_Season41
  }
  NamedRerouteDeclaration(None=@n47.G) {name=GreenChannelPolarTexture, nodeColor=<rgba 0 1 0 1>}
  n48 = NamedRerouteUsage()
  n49 = NamedRerouteUsage()
  n50 = NamedRerouteUsage()
  n51 = NamedRerouteUsage()
  n52 = Add(A=@n32, B=@n33)
  NamedRerouteDeclaration(None=LinearInterpolate(A=Multiply(A=@n48, B=@n49), B=Add(A=@n50, B=@n51), Alpha=@n52)) {name=MainColours, nodeColor=<rgba 1 0.2344 0 1>}
  NamedRerouteDeclaration(None=FunctionInput() {inputName=In_MaterialAttributes, inputType=FunctionInput_MaterialAttributes}) {name=In_MaterialAttributes, nodeColor=<rgba 0.7109 0 1 1>}
  NamedRerouteDeclaration(None=FunctionInput() {inputName=In_TechArtMask, sortPriority=1}) {name=In_TechArtMask, nodeColor=<rgba 0 1 0.3906 1>}
  n53 = TextureObjectParameter() {
   group=CopperMoggy
   parameterName=SoftNoise
   samplerType=SAMPLERTYPE_LinearColor
   texture=&HighContrastNoise
  }
  NamedRerouteDeclaration(None=@n53) {name=SoftNoise, nodeColor=<rgba 0.1719 0 1 1>}
  n54 = NamedRerouteUsage()
  n55 = NamedRerouteUsage()
  n56 = Power(
   Base=Multiply(
    A=TextureSample!(UVs=@n7, Tex=@n54, Apply View MipBias) {samplerSource=SSM_Wrap_WorldGroupSettings, texture=/Game/Global/Textures/Default/Noise/3xPerlin_M}.R
    B=TextureSample(UVs=@n14, Tex=@n55) {samplerSource=SSM_Wrap_WorldGroupSettings, texture=&3xPerlin_M}.G
   )
   Exp=ScalarParameter() {defaultValue=1, group=CopperMoggy, parameterName=SoftNoisePower}
  )
  NamedRerouteDeclaration(None=SmoothStep(Value=Saturate(None=@n56)) {constMax=0.5}) {name=AnimatedSoftNoise, nodeColor=<rgba 0.1484 1 0 1>}
  n57 = Time()
  n58 = ScalarParameter() {defaultValue=0.1, parameterName=BlueChannelRotationSpeed}
  n59 = TextureCoordinate() {coordinateIndex=1}
  n60 = TextureSample!(
   UVs=Rotator(Coordinate=TextureCoordinate() {coordinateIndex=1}, Time=Multiply(A=@n57, B=@n58))
   Tex
   "DDX(UVs)"=DDX(Value=@n59)
   "DDY(UVs)"=DDY(Value=@n59)
   Apply View MipBias
  ) {
   mipValueMode=TMVM_Derivative
   samplerSource=SSM_Clamp_WorldGroupSettings
   samplerType=SAMPLERTYPE_LinearColor
   texture=&T_Glider_Season41
  }
  n61 = Time()
  n62 = Multiply(A=Sine(None=Multiply(A=@n61) {constB=0.333})) {constB=0.1}
  NamedRerouteDeclaration(None=Saturate(None=Add(A=@n60.B, B=@n62))) {name=BlueChannelPlanarTexture, nodeColor=<rgba 0 0 1 1>}
  n63 = Multiply(A=NamedRerouteUsage(), B=VectorParameter() {parameterName=PatternColor, defaultValue=<rgba 0.9131 0.7835 0.0908 1>}.RGB)
  n64 = OneMinus(None=SphereMask(A=TextureCoordinate() {coordinateIndex=1}, B=Constant2Vector() {g=0.5, r=0.5}) {attenuationRadius=0.55, hardnessPercent=80})
  NamedRerouteDeclaration(None=Multiply(A=Multiply(A=@n63, B=@n64)) {constB=1.33}) {name=HalfTonePattern, nodeColor=<rgba 0.8516 1 0 1>}
  n65 = NamedRerouteUsage()
  n66 = NamedRerouteUsage()
  n67 = Add(A=NamedRerouteUsage(), B=NamedRerouteUsage())
  NamedRerouteDeclaration(None=LinearInterpolate(A=Multiply(A=@n65, B=@n66), B=Saturate(None=@n67), Alpha=@n52)) {name=CombinedTexture, nodeColor=<rgba 0 1 0.9297 1>}
  n68 = ScalarParameter() {parameterName=Min}
  n69 = ScalarParameter() {defaultValue=1, parameterName=Max}
  n70 = NamedRerouteUsage()
  NamedRerouteDeclaration(None=MaterialFunctionCall!(
    A=VectorParameter() {parameterName=ColorA, defaultValue=<rgba 0.087 0.049 0.0355 1>}.RGB
    B=VectorParameter() {parameterName=ColorB, defaultValue=<rgba 0.287 0.0706 0.0078 1>}.RGB
    C=VectorParameter() {parameterName=ColorC, defaultValue=<rgba 2 0.6211 0.3652 0>}.RGB
    D=VectorParameter() {parameterName=ColorD, defaultValue=<rgba 2 1.474 0.3652 0>}.RGB
    Alpha=SmoothStep(Min=@n68, Max=@n69, Value=@n70)
   ) {materialFunction=/Game/Environments/FrontEnd/BackPlates/Textures/Delulu/Lerp_4Color}.Result) {name=CombinedColor, nodeColor=<rgba 1 0.6094 0 1>}
  NamedRerouteDeclaration(None=@n16.Roughness) {name=In_Roughness, nodeColor=<rgba 1 0 0.375 1>}
  NamedRerouteDeclaration(None=@n16.Specular) {name=In_Specular, nodeColor=<rgba 0 0.6953 1 1>}
  n71 = Time()
  n72 = ScalarParameter() {defaultValue=0.45, parameterName=AlphaChannelRotationSpeed}
  n73 = TextureCoordinate() {coordinateIndex=1}
  n74 = TextureSample(
   UVs=Rotator(Coordinate=TextureCoordinate() {coordinateIndex=1}, Time=Multiply(A=@n71, B=@n72))
   "DDX(UVs)"=DDX(Value=@n73)
   "DDY(UVs)"=DDY(Value=@n73)
  ) {
   mipValueMode=TMVM_Derivative
   samplerSource=SSM_Clamp_WorldGroupSettings
   samplerType=SAMPLERTYPE_LinearColor
   texture=&T_Glider_Season41
  }
  NamedRerouteDeclaration(None=@n74.A) {name=AlphaChannelPlanarTexture, nodeColor=<rgba 1 1 1 1>}
  NamedRerouteDeclaration(None=MaterialFunctionCall(
    A=VectorParameter() {parameterName=RedChannelColorA, defaultValue=<rgba 0.087 0.049 0.0355 1>}.RGB
    B=VectorParameter() {parameterName=RedChannelColorB, defaultValue=<rgba 0.287 0.0706 0.0078 1>}.RGB
    C=VectorParameter() {parameterName=RedChannelColorC, defaultValue=<rgba 2 1.474 0.3652 0>}.RGB
    Alpha=NamedRerouteUsage()
   ) {materialFunction=&Lerp_3Color}.Result) {name=RedChannelColors, nodeColor=<rgba 0.8047 0 1 1>}
  NamedRerouteDeclaration(None=MaterialFunctionCall(
    A=VectorParameter() {parameterName=GreenChannelColorA, defaultValue=<rgba 0.4522 0.3013 0.2949 1>}.RGB
    B=VectorParameter() {parameterName=GreenChannelColorB, defaultValue=<rgba 2 0.4924 0.0546 1>}.RGB
    C=VectorParameter() {parameterName=GreenChannelColorC, defaultValue=<rgba 4 2.948 0.7304 0>}.RGB
    Alpha=NamedRerouteUsage()
   ) {materialFunction=&Lerp_3Color}.Result) {name=GreenChannelColors, nodeColor=<rgba 0.8047 0 1 1>}
  NamedRerouteDeclaration(None=MaterialFunctionCall(
    A=VectorParameter() {parameterName=BlueChannelColorA, defaultValue=<rgba 0.2348 0.0546 0.047 1>}.RGB
    B=VectorParameter() {parameterName=BlueChannelColorB, defaultValue=<rgba 2 0.4924 0.0546 1>}.RGB
    C=VectorParameter() {parameterName=BlueChannelColorC, defaultValue=<rgba 3 2.07 0.1094 0>}.RGB
    Alpha=NamedRerouteUsage()
   ) {materialFunction=&Lerp_3Color}.Result) {name=BlueChannelColors, nodeColor=<rgba 0.8047 0 1 1>}
  NamedRerouteDeclaration(None=MaterialFunctionCall(
    A=VectorParameter() {parameterName=AlphaChannelColorA, defaultValue=<rgba 0.1478 0.077 0.0144 1>}.RGB
    B=VectorParameter() {parameterName=AlphaChannelColorB, defaultValue=<rgba 1.6 0.2061 0.0437 1>}.RGB
    C=VectorParameter() {parameterName=AlphaChannelColorC, defaultValue=<rgba 2 1.2765 0.3021 0>}.RGB
    Alpha=NamedRerouteUsage()
   ) {materialFunction=&Lerp_3Color}.Result) {name=AlphaChannelColors, nodeColor=<rgba 0.8047 0 1 1>}
  n75 = TextureSample!(UVs=TextureCoordinate(), Tex, Apply View MipBias) {
   samplerSource=SSM_Clamp_WorldGroupSettings
   texture=/BRCosmetics/Gadgets/Assets/VinderTech_GliderChute/Glider_Season_41/Skins/Honor/Textures/T_Glider_Season41_EdgeGlow
  }
  n76 = Sine(None=Multiply(A=Time()) {constB=2})
  n77 = Multiply(A=@n75.RGB, B=Add(B=Multiply(A=@n76) {constB=0.2}) {constA=1})
  NamedRerouteDeclaration(None=Saturate(None=Multiply(A=@n77) {constB=0.5})) {name=EdgeGlow, nodeColor=<rgba 0 0 1 1>}
  NamedRerouteUsage()
  NamedRerouteUsage()
  NamedRerouteUsage()
  NamedRerouteUsage()
  NamedRerouteUsage()
  NamedRerouteUsage()
  n78 = MaterialFunctionCall!(Position=Add(A=ActorPositionWS(), B=Constant3Vector() {constant=<rgba 0 0 130 0>}), UVRadialScale) {materialFunction=/Game/Weapons/WeaponSkins/MaterialFunction/ObjectStabilizedScreenSpaceUVs}
  n79 = ScalarParameter() {defaultValue=80, group=Flame, parameterName=PatternTiling}
  n80 = TextureSample(UVs=Multiply(
    A=NamedRerouteDeclaration(None=@n78.StabilizedUVs) {name=ScreenSpaceUVs, nodeColor=<rgba 0.7109 1 0 1>}
    B=NamedRerouteDeclaration(None=@n79) {name=PatternTIling, nodeColor=<rgba 0.5234 1 0 1>}
   )) {
   samplerSource=SSM_Wrap_WorldGroupSettings
   samplerType=SAMPLERTYPE_LinearGrayscale
   texture=/BRCosmetics/Weapons/FORT_Melee/Pickaxe_ScoopPluck/Textures/T_DottedPattern_DF
  }
  n81 = Subtract(A=ScalarParameter() {defaultValue=0.1, group=Flame, parameterName=MinDotSize})
  n82 = Subtract(A=ScalarParameter() {defaultValue=1.3, group=Flame, parameterName=MaxDotSize})
  n83 = Saturate(None=@n31)
  n84 = Saturate(None=Add(
    A=NamedRerouteDeclaration(None=@n80.R) {name=PatternDistanceField, nodeColor=<rgba 1 0 0.3984 1>}
    B=LinearInterpolate(A=@n81, B=@n82, Alpha=@n83) {constA=-1, constAlpha=1, constB=0.3}
   ))
  n85 = Subtract(A=@n84) {constB=0.5}
  n86 = DDX(Value=@n84)
  n87 = DDY(Value=@n84)
  n88 = Max(A=Add(A=Abs(None=@n86), B=Abs(None=@n87))) {constB=0}
  Saturate(None=Add(A=Divide(A=@n85, B=@n88)) {constB=0.5})
 ]}
